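# Unlit translucent material for animated KelpLinen cartoon cards, with hue shift, time-of-day glow and location-based opacity fade.
# UV math picks an atlas cell by index, with optional sine and noise warping.
# World position offset chains optional camera facing, oscillation, rotation of vertex-color RGB appendages around pivots, and sine bending.
# A parameter-collection velocity adds limb sway and tilt, while a debug switch previews vertex-paint masks.
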Material /Clyde/Environments/KelpLinen/Materials/Cartoons/M_Cartoon
graph {
 properties {
  bAutomaticallySetUsageInEditor=false
  bEnableResponsiveAA=true
  bOutputTranslucentVelocity=true
  bUseMaterialAttributes=true
  bUseTranslucencyVertexFog=false
  blendMode=BLEND_Translucent
  opacityMaskClipValue=0.05
  shadingModel=MSM_Unlit
  usedShadingModels=Unlit
 }
 nodes [
  n1 = ScalarParameter() {defaultValue=2, parameterName=Image.NumRowColumns}
  n2 = Divide(A=ScalarParameter() {parameterName=Image.Index, sortPriority=20}, B=Multiply(A=@n1, B=@n1))
  n3 = ScalarParameter() {defaultValue=2, parameterName=Image.NumRowColumns}
  n4 = ScalarParameter() {defaultValue=2, parameterName=Image.NumRowColumns}
  n5 = Divide(A=ScalarParameter() {defaultValue=2048, parameterName=Image.TextureSize}, B=@n4)
  n6 = Divide(A=@n5)
  n7 = ScalarParameter() {parameterName=Image.Index, sortPriority=20}
  n8 = Add(A=@n6, B=Multiply(A=@n5, B=Modulo!(A=@n7, B=@n4)))
  n9 = Divide(A=@n7, B=ScalarParameter() {defaultValue=2, parameterName=Image.NumRowColumns})
  n10 = Add(A=@n6, B=Multiply(A=@n5, B=Floor(None=@n9)))
  n11 = MaterialFunctionCall!(
   Texture Scale=ScalarParameter() {defaultValue=1, parameterName=Image.UVScale}
   UVs=TextureCoordinate()
   Offset XY Texturesize Z=AppendVector(A=AppendVector(A=@n8, B=@n10), B=ScalarParameter() {defaultValue=2048, parameterName=Image.TextureSize})
  ) {materialFunction=/Engine/Functions/Engine_MaterialFunctions01/Texturing/ScaleUVsAroundPoint}
  n12 = ScalarParameter() {parameterName=Image.Offset.X}
  n13 = ScalarParameter() {parameterName=Image.Offset.Y}
  n14 = ComponentMask(None=TextureCoordinate()) {r=true}
  n15 = NamedRerouteUsage()
  n16 = ScalarParameter() {defaultValue=0.02, parameterName=SineUV.WaveX.Intensity, sortPriority=50}
  n17 = ScalarParameter() {defaultValue=0.02, parameterName=SineUV.WaveY.Intensity, sortPriority=50}
  n18 = ScalarParameter() {defaultValue=2, parameterName=Image.NumRowColumns}
  n19 = Multiply(
   A=Sine(None=Add(A=@n14, B=@n15))
   B=Multiply(
    A=MaterialFunctionCall(X=@n16, Y=@n17) {materialFunction=&MakeFloat2}.Result
    B=Divide(B=@n18)
   )
  ) {constB=0.02}
  n20 = ScalarParameter() {parameterName=SineUV.UseSineUVOoscillation, sortPriority=50}
  n21 = TextureSample(UVs=Panner(Coordinate=TextureCoordinate(), Time=NamedRerouteUsage()) {speedX=0.33, speedY=1}) {samplerType=SAMPLERTYPE_Masks, texture=/Game/Characters/Player/Female/Medium/Bodies/F_MED_Neon_Tight_Suit/Materials/Soft_Perlin_output}
  n22 = ScalarParameter() {defaultValue=2, parameterName=Image.NumRowColumns}
  n23 = Multiply(A=@n21.R, B=Multiply(A=ScalarParameter() {defaultValue=0.01, parameterName=NoiseWarp.Amount, sortPriority=40}, B=Divide(B=@n22))) {constB=0.01}
  n24 = ScalarParameter() {parameterName=NoiseWarp.UseNoiseWarp, sortPriority=40}
  Add(
   A=Add(
    A=MaterialFunctionCall("Animation  Phase (0-1)"=@n2, Number of Rows=@n3, Number of Columns=@n3, UVs=@n11.UVs) {materialFunction=&FlipBook}.UVs
    B=AppendVector(A=@n12, B=@n13)
   )
   B=Add(A=LinearInterpolate(B=@n19, Alpha=@n20), B=LinearInterpolate(B=@n23, Alpha=@n24))
  )
  n25 = Time()
  n26 = ScalarParameter() {parameterName=Timing.CustomOffset}
  n27 = Frac(None=MaterialFunctionCall(f3=ActorPositionWS()) {materialFunction=&AddComponents}.f3)
  Add(A=@n25, B=If(A=@n26, "A > B"=@n26, "A == B"=Multiply(A=@n27) {constB=5}, "A < B"=@n26) {equalsThreshold=0.01})
  n28 = Multiply(A=Add(A=@n25, B=@n26), B=ScalarParameter() {defaultValue=4, parameterName=Timing.Speed})
  n29 = ScalarParameter() {defaultValue=4, parameterName=Timing.NumberOfSteps}
  n30 = Divide(B=@n29)
  n31 = Multiply(A=Frac(None=Multiply(A=@n28, B=@n30)), B=@n29) {constB=4}
  n32 = Divide(A=@n28) {constB=10}
  LinearInterpolate(A=@n32, B=Divide(A=Floor(None=@n31), B=@n29) {constB=4}, Alpha=ScalarParameter() {parameterName=UseSteppedTime})
  NamedRerouteDeclaration(None=@n32) {name=LoopingTime, nodeColor=<rgba 1 0 0.5859 1>}
  NamedRerouteDeclaration(None=@n26) {name=PerActorRandom, nodeColor=<rgba 0.0078 0 1 1>}
  n33 = MaterialFunctionCall(Float2=StaticSwitchParameter(True=TextureCoordinate() {coordinateIndex=1}, False=TextureCoordinate()) {defaultValue=true, parameterName=WPO.UVChannel2}) {materialFunction=&BreakOutFloat2Components}
  n34 = Subtract(A=AppendVector(A=@n33.R, B=OneMinus(None=@n33.G))) {constB=0.5}
  n35 = Max(A=SquareRoot(None=DotProduct(A=@n34, B=@n34))) {constB=0}
  n36 = WorldPosition()
  n37 = ObjectPositionWS()
  n38 = AppendVector(A=Multiply(A=Divide(A=@n34, B=@n35), B=Distance(A=@n36.XYZ, B=@n37)), B=Constant())
  n39 = Transform(None=@n38) {transformSourceType=TRANSFORMSOURCE_Camera}
  NamedRerouteDeclaration(None=@n39) {name=CameraToWorldVector, nodeColor=<rgba 0.125 1 0 1>}
  NamedRerouteDeclaration(None=Transform(None=@n38) {transformSourceType=TRANSFORMSOURCE_Camera, transformType=TRANSFORM_Local}) {name=CameraToLocalVector, nodeColor=<rgba 0.125 1 0 1>}
  n40 = Multiply(A=Time(), B=ScalarParameter() {defaultValue=0.4, parameterName=VelocityWPO.Sway.InfluenceVariationSpeed})
  n41 = ScalarParameter() {defaultValue=0.01, parameterName=VelocityWPO.Sway.InfluenceVariationOffset}
  n42 = ConstantBiasScale(None=Sine(None=Add(A=@n40, B=@n41)) {period=1.33})
  n43 = Divide(A=@n41)
  n44 = ConstantBiasScale(None=Sine(None=Add(A=@n40, B=@n43)) {period=1.77})
  n45 = LinearInterpolate(
   A=ScalarParameter() {defaultValue=1, parameterName=VelocityWPO.Sway.Variation.Min, sliderMax=1}
   B=ScalarParameter() {defaultValue=1, parameterName=VelocityWPO.Sway.Variation.Max, sliderMax=1}
   Alpha=Divide(A=Add(A=@n42, B=@n44))
  )
  NamedRerouteDeclaration(None=@n45) {name=SwayVariation, nodeColor=<rgba 0 1 0.7422 1>}
  n46 = WorldPosition()
  n47 = MaterialFunctionCall() {materialFunction=&ObjectPivotPoint}
  n48 = ScalarParameter() {defaultValue=0.1, desc=Smooth Blend via distance, group=WPO, parameterName=WPO.Appendage.Green.Distance}
  n49 = MaterialFunctionCall() {materialFunction=&ObjectScale}
  Saturate(None=Divide(A=Distance(A=@n46.XYZ, B=@n47.Object Pivot Location), B=Multiply(A=@n48, B=@n49.Scale X))) {desc=Smooth Blend via distance}
  n50 = WorldPosition()
  n51 = VectorParameter() {group=WPO, parameterName=WPO.Appendage.Red.PivotPointOffset, defaultValue=<rgba 0 0 0 1>}
  n52 = Normalize(VectorInput=Subtract(A=CameraPositionWS(), B=ObjectPositionWS()))
  n53 = Add(
   A=MaterialFunctionCall() {materialFunction=&ObjectPivotPoint}.Object Pivot Location
   B=Multiply(
    A=MaterialFunctionCall!(VectorToRotate=@n51.RGB, LookAtVector=@n52, RestingVector) {materialFunction=/Engine/Functions/Engine_MaterialFunctions01/Vectors/RotateVector}.Result
    B=MaterialFunctionCall() {materialFunction=&ObjectScale}.Scale X
   )
  )
  n54 = ScalarParameter() {defaultValue=0.1, desc=Smooth Blend via distance, group=WPO, parameterName=WPO.Appendage.Red.Distance}
  n55 = MaterialFunctionCall() {materialFunction=&ObjectScale}
  Saturate(None=Divide(A=Distance(A=@n50.XYZ, B=@n53), B=Multiply(A=@n54, B=@n55.Scale X))) {desc=Smooth Blend via distance}
  n56 = WorldPosition()
  n57 = Add(
   A=MaterialFunctionCall() {materialFunction=&ObjectPivotPoint}.Object Pivot Location
   B=Multiply(
    A=VectorParameter() {group=WPO, parameterName=WPO.Appendage.Blue.PivotPointOffset, defaultValue=<rgba 0 0 0 1>}.RGB
    B=MaterialFunctionCall() {materialFunction=&ObjectScale}.Scale X
   )
  )
  n58 = ScalarParameter() {defaultValue=0.1, desc=Smooth Blend via distance, group=WPO, parameterName=WPO.Appendage.Blue.Distance}
  n59 = MaterialFunctionCall() {materialFunction=&ObjectScale}
  Saturate(None=Divide(A=Distance(A=@n56.XYZ, B=@n57), B=Multiply(A=@n58, B=@n59.Scale X))) {desc=Smooth Blend via distance}
  n60 = Time()
  n61 = ScalarParameter() {parameterName=HueShiftSpeed}
  n62 = TextureSampleParameter2D() {parameterName=CartoonImage, texture=&/T_KelpLinen_CartoonSet_01}
  n63 = Multiply(
   A=ScalarParameter() {defaultValue=1, parameterName=EmissiveBoost}
   B=MaterialFunctionCall(Hue Shift Percentage=Multiply(A=@n60, B=@n61), Texture=@n62.RGBA) {materialFunction=&HueShift}.Result
  )
  n64 = VectorParameter() {group=Glow, parameterName=Glow.Color.Day, defaultValue=<rgba 1 0.1793 0 1>}
  n65 = VectorParameter() {group=Glow, parameterName=Glow.Color.Night, defaultValue=<rgba 1 0.1793 0 1>}
  n66 = LinearInterpolate(A=@n64.RGB, B=@n65.RGB)
  n67 = TextureCoordinate() {coordinateIndex=1}
  n68 = AppendVector(
   A=ScalarParameter() {defaultValue=0.5, group=Glow, parameterName=Glow.Pos.X}
   B=ScalarParameter() {defaultValue=1, group=Glow, parameterName=Glow.Pos.Y}
  )
  n69 = ScalarParameter() {defaultValue=0.5, group=Glow, parameterName=Glow.Radius}
  n70 = ScalarParameter() {defaultValue=100, group=Glow, parameterName=Glow.Hardness}
  n71 = LinearInterpolate(
   A=@n63
   B=LinearInterpolate(
    A=@n63
    B=MaterialFunctionCall(Color_Morning=@n66, Color_Day=@n64.RGB, Color_Evening=@n66, Color_Night=@n65.RGB) {materialFunction=&Color_Time_of_Day}.Result
    Alpha=SphereMask(A=@n67, B=@n68, Radius=@n69, Hardness=@n70)
   )
   Alpha=ScalarParameter() {defaultValue=1, group=Glow, parameterName=Glow.Opacity}
  )
  n72 = ShadowReplace(Default=ComponentMask(None=@n62.RGBA) {a=true}, Shadow=Constant())
  n73 = Subtract(A=@n57, B=CameraPositionWS())
  n74 = NamedRerouteUsage()
  n75 = ScalarParameter() {defaultValue=0.2, group=WPO, parameterName=WPO.Appendage.Blue.RotationOffset}
  n76 = Sine(None=Multiply(A=Add(A=@n74, B=@n75), B=ScalarParameter() {defaultValue=0.2, group=WPO, parameterName=WPO.Appendage.Blue.RotationSpeed}))
  n77 = ScalarParameter() {defaultValue=0.01, group=WPO, parameterName=WPO.Appendage.Blue.RotationAmount}
  n78 = Normalize(VectorInput=Subtract(A=@n47.Object Pivot Location, B=CameraPositionWS()))
  n79 = Add(A=NamedRerouteUsage(), B=ScalarParameter() {defaultValue=0.7, group=WPO, parameterName=WPO.Appendage.Green.RotationOffset})
  n80 = ScalarParameter() {defaultValue=0.2, group=WPO, parameterName=WPO.Appendage.Green.RotationSpeed}
  n81 = Multiply(A=Sine(None=Multiply(A=@n79, B=@n80)), B=ScalarParameter() {defaultValue=0.01, group=WPO, parameterName=WPO.Appendage.Green.RotationAmount}) {constB=0.1}
  n82 = TransformPosition(None=Add(
    A=LocalPosition() {localOrigin=Primitive}.XYZ
    B=VectorParameter() {group=WPO, parameterName=WPO.Appendage.Green.PivotPointOffset, defaultValue=<rgba 0 0 0 1>}.RGB
   )) {transformType=TRANSFORMPOSSOURCE_World}
  n83 = Normalize(VectorInput=Subtract(A=@n53, B=CameraPositionWS()))
  n84 = NamedRerouteUsage()
  n85 = ScalarParameter() {defaultValue=0.2, group=WPO, parameterName=WPO.Appendage.Red.RotationSpeed}
  n86 = Multiply(A=Sine(None=Multiply(A=@n84, B=@n85)), B=ScalarParameter() {defaultValue=0.01, group=WPO, parameterName=WPO.Appendage.Red.RotationAmount}) {constB=0.1}
  n87 = Add(
   A=MaterialFunctionCall() {materialFunction=&ObjectPivotPoint}.Object Pivot Location
   B=Multiply(
    A=VectorParameter() {group=WPO, parameterName=WPO.PivotPointOffset, defaultValue=<rgba 0 0 0 1>}.RGB
    B=MaterialFunctionCall() {materialFunction=&ObjectScale}.Scale X
   )
  )
  n88 = Subtract(A=@n87, B=CameraPositionWS())
  n89 = NamedRerouteUsage()
  n90 = ScalarParameter() {defaultValue=0.2, group=WPO, parameterName=WPO.Oscillation.RotationSpeed}
  n91 = Multiply(A=Sine(None=Multiply(A=@n89, B=@n90)), B=ScalarParameter() {defaultValue=0.05, group=WPO, parameterName=WPO.Oscillation.RotationAmount}) {constB=0.1}
  n92 = ScalarParameter() {group=WPO, parameterName=WPO.Rotation}
  n93 = WorldPosition()
  n94 = Add(A=Time(), B=ScalarParameter() {group=WPO, parameterName=WPO.Oscillation.Offset})
  n95 = VectorParameter() {group=WPO, parameterName=WPO.Oscillaltion, channelNames {a=Distance Y, b=Distance X, g=Speed Y, r=Speed X}, defaultValue=<rgba 0 0.2 0 0>}
  n96 = AppendVector(A=@n95.Speed X, B=@n95.Speed X)
  n97 = MaterialFunctionCall() {materialFunction=&ObjectScale}
  n98 = MaterialFunctionCall(Float2=Multiply(A=Multiply(A=@n96, B=@n97.Scale X), B=ScalarParameter() {defaultValue=1, group=WPO, parameterName=WPO.Oscillation.SpeedMulti})) {materialFunction=&BreakOutFloat2Components}
  n99 = Sine(None=Multiply(A=@n94, B=@n98.R)) {period=1.33}
  n100 = AppendVector(A=@n95.Distance X, B=@n95.Distance X)
  n101 = MaterialFunctionCall(Float2=Multiply(A=Multiply(A=@n100, B=@n97.Scale X), B=ScalarParameter() {defaultValue=1, group=WPO, parameterName=WPO.Oscillation.DistanceMulti})) {materialFunction=&BreakOutFloat2Components}
  n102 = Sine(None=Multiply(A=@n94, B=@n98.G))
  n103 = Add(A=@n39, B=ObjectPositionWS())
  n104 = WorldPosition()
  n105 = Add(
   A=StaticSwitchParameter(True=Subtract(A=@n103, B=@n104.XYZ), False=Constant()) {defaultValue=true, parameterName=WPO.UseCameraFacing}
   B=MaterialFunctionCall(Y=Multiply(A=@n99, B=@n101.R) {constB=100}, Z=Multiply(A=@n102, B=@n101.G) {constB=100}) {materialFunction=&MakeFloat3}.Result
  )
  n106 = Add(
   A=@n105
   B=RotateAboutAxis(NormalizedRotationAxis=Normalize(VectorInput=@n88), RotationAngle=Add(A=@n91, B=@n92), PivotPoint=@n87, Position=Add(A=@n105, B=@n93.XYZ))
  )
  n107 = Add(A=@n106, B=@n50.XYZ)
  n108 = VertexColor()
  n109 = ComponentMask(None=TextureCoordinate() {coordinateIndex=2}) {r=true}
  n110 = Add(
   A=@n106
   B=LinearInterpolate(B=RotateAboutAxis(NormalizedRotationAxis=@n83, RotationAngle=@n86, PivotPoint=@n53, Position=@n107), Alpha=Multiply(A=@n108.R, B=@n109))
  )
  n111 = Add(A=@n110, B=@n46.XYZ)
  n112 = VertexColor()
  n113 = ComponentMask(None=TextureCoordinate() {coordinateIndex=2}) {r=true}
  n114 = Add(
   A=@n110
   B=LinearInterpolate(B=RotateAboutAxis(NormalizedRotationAxis=@n78, RotationAngle=@n81, PivotPoint=@n82, Position=@n111), Alpha=Multiply(A=@n112.G, B=@n113))
  )
  n115 = TextureCoordinate() {coordinateIndex=2}
  n116 = LinearInterpolate(
   B=RotateAboutAxis(
    NormalizedRotationAxis=Normalize(VectorInput=@n73)
    RotationAngle=Multiply(A=@n76, B=@n77) {constB=0.1}
    PivotPoint=@n57
    Position=Add(A=@n114, B=@n56.XYZ)
   )
   Alpha=Multiply(A=VertexColor().B, B=ComponentMask(None=@n115) {r=true})
  )
  n117 = GetMaterialAttributes(None=SetMaterialAttributes!(
    MaterialAttributes=SetMaterialAttributes!(MaterialAttributes, Emissive Color=@n71, Opacity=@n72, Opacity Mask=@n72) {attributeSetTypes [#1, #2, #3]}
    World Position Offset=Add(A=@n114, B=@n116)
   ) {attributeSetTypes [#4]}) {attributeGetTypes [#2]}
  n118 = MaterialFunctionCall() {materialFunction=/ClydeCosmetics/Environments/KelpLinen/MF_KelpLinen_LocationFade}
  n119 = GetMaterialAttributes(None=SetMaterialAttributes!(MaterialAttributes=@n117.MaterialAttributes, Opacity=Multiply(A=@n117.Opacity, B=@n118.Output)) {attributeSetTypes [#2]}) {attributeGetTypes [#4]}
  n120 = Transform(None=Constant3Vector() {constant=<rgba 1 1 0 0>}) {transformSourceType=TRANSFORMSOURCE_Camera}
  n121 = ComponentMask(None=TextureCoordinate() {coordinateIndex=2}) {r=true}
  n122 = ScalarParameter() {defaultValue=2, group=WPO, parameterName=WPO.Sine.Tiling}
  n123 = Time()
  n124 = ScalarParameter() {defaultValue=0.2, group=WPO, parameterName=WPO.Sine.Speed}
  n125 = Sine(None=Add(A=Multiply(A=@n121, B=@n122), B=Multiply(A=@n123, B=@n124) {constB=0.2}))
  n126 = ScalarParameter() {group=WPO, parameterName=WPO.SineMulti}
  n127 = MaterialFunctionCall() {materialFunction=&ObjectScale}
  n128 = Desaturation(None=Multiply(A=VertexColor(), B=VectorParameter() {group=WPO, parameterName=WPO.Sine.VertexColorChannels, defaultValue=<rgba 0 0 1 1>}.RGB))
  n129 = Multiply(
   A=Multiply(A=Saturate(None=@n125), B=Multiply(A=@n126, B=@n127.Scale X))
   B=StaticSwitchParameter(True=Constant() {r=1}, False=@n128) {group=WPO, parameterName=WPO.Sine.IgnoreVertexPaint}
  )
  n130 = SetMaterialAttributes!(
   MaterialAttributes=@n119.MaterialAttributes
   World Position Offset=Add(A=@n119.World Position Offset, B=Multiply(A=@n120, B=@n129)) {constB=0}
  ) {attributeSetTypes [#4]}
  n131 = GetMaterialAttributes(None=@n130) {attributeGetTypes [#1]}
  n132 = VertexInterpolator(VS=LocalPosition() {localOrigin=Primitive}.XYZ)
  n133 = Constant3Vector() {constant=<rgba 1 1 1 0>}
  n134 = Multiply(A=VectorParameter() {group=WPO, parameterName=WPO.Appendage.Red.PivotPointOffset, defaultValue=<rgba 0 0 0 1>}.RGB, B=@n133)
  n135 = Multiply(
   A=MaterialFunctionCall() {materialFunction=&ObjectScale}.Scale X
   B=ScalarParameter() {defaultValue=1, parameterName=DebugRadius}
  )
  n136 = SphereMask(
   A=TransformPosition(None=Add(A=@n132.PS, B=@n134)) {transformType=TRANSFORMPOSSOURCE_World}
   B=TransformPosition(None=Constant3Vector()) {transformType=TRANSFORMPOSSOURCE_World}
   Radius=@n135
  )
  n137 = Multiply(A=@n133, B=VectorParameter() {group=WPO, parameterName=WPO.Appendage.Green.PivotPointOffset, defaultValue=<rgba 0 0 0 1>}.RGB)
  n138 = SphereMask(
   A=TransformPosition(None=Add(A=@n132.PS, B=@n137)) {transformType=TRANSFORMPOSSOURCE_World}
   B=TransformPosition(None=Constant3Vector()) {transformType=TRANSFORMPOSSOURCE_World}
   Radius=@n135
  )
  n139 = Multiply(A=@n133, B=VectorParameter() {group=WPO, parameterName=WPO.Appendage.Blue.PivotPointOffset, defaultValue=<rgba 0 0 0 1>}.RGB)
  n140 = SphereMask(
   A=TransformPosition(None=Add(A=@n132.PS, B=@n139)) {transformType=TRANSFORMPOSSOURCE_World}
   B=TransformPosition(None=Constant3Vector()) {transformType=TRANSFORMPOSSOURCE_World}
   Radius=@n135
  )
  Saturate(None=Add(
    A=Multiply(A=@n131.Emissive Color) {constB=0.25}
    B=MaterialFunctionCall(X=@n136, Y=@n138, Z=@n140) {materialFunction=&MakeFloat3}.Result
   ))
  n141 = GetMaterialAttributes(None=@n130) {attributeGetTypes [#4]}
  n142 = CollectionParameter() {collection=/ClydeCosmetics/Environments/KelpLinen/Materials/MPC_KelpLinen, parameterName=Cartoons_WPOVelocity}
  n143 = TextureCoordinate() {coordinateIndex=2}
  n144 = ScalarParameter() {defaultValue=1, desc="Controls how much of the limb is effected, high numbers = less effected", parameterName=VelocityWPO.Sway.Contrast}
  n145 = LinearInterpolate(
   B=Multiply(
    A=ComponentMask(None=@n142) {b=true, g=true, r=true}
    B=CollectionParameter() {collection=&MPC_KelpLinen, parameterName=Cartoons_WPOVelocityMulti}
   )
   Alpha=Power(Base=ComponentMask(None=@n143) {r=true}, Exp=@n144)
  )
  n146 = ScalarParameter() {defaultValue=400, parameterName=VelocityWPO.Sway.MaxDistance}
  n147 = Multiply(A=@n146) {constB=-1}
  n148 = LinearInterpolate(
   B=Multiply(A=Clamp(None=@n145, Min=@n147, Max=@n146), B=@n45)
   Alpha=ScalarParameter() {defaultValue=1, parameterName=VelocityWPO.Sway.Influence, sliderMax=1}
  )
  n149 = GetMaterialAttributes(None=SetMaterialAttributes(MaterialAttributes=@n141.MaterialAttributes, World Position Offset=Add(A=@n141.World Position Offset, B=@n148) {constB=0}) {attributeSetTypes [#4]}) {attributeGetTypes [#4]}
  n150 = Add(
   A=MaterialFunctionCall() {materialFunction=&ObjectPivotPoint}.Object Pivot Location
   B=Multiply(
    A=VectorParameter() {group=WPO, parameterName=WPO.PivotPointOffset, defaultValue=<rgba 0 0 0 1>}.RGB
    B=MaterialFunctionCall() {materialFunction=&ObjectScale}.Scale X
   )
  )
  n151 = Normalize(VectorInput=Subtract(A=@n150, B=CameraPositionWS()))
  n152 = ComponentMask(None=CollectionParameter() {collection=&MPC_KelpLinen, parameterName=Cartoons_WPOVelocity}) {b=true, g=true, r=true}
  n153 = MaterialFunctionCall(Vector 3=@n152) {materialFunction=/Engine/Functions/Engine_MaterialFunctions02/Utility/VectorLength}
  n154 = Divide(
   A=If(A=@n153.V3 Length, "A > B"=@n153.V3 Length, "A == B"=Constant3Vector() {constant=<rgba 1 0 0 0>}, "A < B"=@n153.V3 Length) {equalsThreshold=0.001}
   B=ScalarParameter() {defaultValue=5, parameterName=VelocityWPO.Rotation.MaxVectorLength}
  )
  n155 = DotProduct(A=Normalize(VectorInput=@n152), B=Normalize(VectorInput=Constant3Vector() {constant=<rgba 1 -1 0 0>}))
  n156 = ScalarParameter() {defaultValue=0.5, parameterName=VelocityWPO.Rotation.MaxRotation}
  n157 = Multiply(A=LinearInterpolate(Alpha=Saturate(None=@n154)), B=Multiply(A=Multiply(A=@n155, B=@n156) {constB=0.5}, B=NamedRerouteUsage()))
  n158 = Add(A=@n149.World Position Offset, B=WorldPosition().XYZ)
  SetMaterialAttributes(
   MaterialAttributes=@n149.MaterialAttributes
   World Position Offset=Add(A=@n149.World Position Offset, B=RotateAboutAxis(NormalizedRotationAxis=@n151, RotationAngle=@n157, PivotPoint=@n150, Position=@n158))
  ) {attributeSetTypes [#4]}
  StaticSwitchParameter(
   True=SetMaterialAttributes!(MaterialAttributes=@n131.MaterialAttributes, Emissive Color=@n128) {attributeSetTypes [#1]}
   False=@n130
  ) {parameterName=Debug}
 ]
}
